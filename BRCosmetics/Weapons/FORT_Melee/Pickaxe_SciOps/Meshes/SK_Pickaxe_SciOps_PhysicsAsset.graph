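# Physics asset for the SciOps pickaxe skeletal mesh, a two-handed melee weapon cosmetic.
# A single capsule body on the Melee_TwoHanded_Handle bone, offset and rotated from the bone origin, is the weapon's only collision shape.
# It has no constraints, so the pickaxe has no jointed or articulated physics parts.
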PhysicsAsset /BRCosmetics/Weapons/FORT_Melee/Pickaxe_SciOps/Meshes/SK_Pickaxe_SciOps_PhysicsAsset
bodies {Melee_TwoHanded_Handle [{shapeType=Capsule, length=59.3506, radius=40.5325, center=<0.7983 6.9071 19.3022>, rotation=<rot -0.9715 -90 0.0003>}]}
constraints []
